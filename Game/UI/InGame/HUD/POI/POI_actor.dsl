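; POI_actor is an actor Blueprint that registers itself as a point-of-interest marker on the Fortnite HUD.
; An enable event adds it through the Fort HUD context with display text, an image and an optional color, and a disable event removes it.
; On destruction it also removes the marker, after checking that the HUD context is still valid.
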(event Custom|Event_EnableMarkerForClient
  (params (DisplayImage :type Texture%202D%20Object%20Reference) (bSpecifyColor :type Boolean) (ImageColor :type Linear%20Color%20Structure) (DisplayText :type Text))
  (HUD|AddPointOfInterest (BlueprintContext|GetFortHUDContext) self DisplayText DisplayImage bSpecifyColor ImageColor))

(event Custom|Event_DisableMarkerForClient
  (HUD|RemovePointOfInterest (BlueprintContext|GetFortHUDContext) self))

(event EventDestroyed
  (bind _returnvalue (BlueprintContext|GetFortHUDContext))
  (Utilities|IsValid _returnvalue
    (:Is%20Valid
      (HUD|RemovePointOfInterest _returnvalue self))))

(fn ConstructionScript)
